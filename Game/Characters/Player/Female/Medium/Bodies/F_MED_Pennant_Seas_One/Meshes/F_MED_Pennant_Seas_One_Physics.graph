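# Physics asset for the F_MED_Pennant_Seas_One female medium player body, covering a dynamic zipper and the upper body around it.
# Capsule bodies sit on spine_05, neck_01, dyn_zipper_MSTR and dyn_zipper, with small capsules for the two zipper bones.
# Constraints chain spine to neck to zipper master to zipper, with tighter swing and twist limits on the final zipper joint.
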PhysicsAsset /Game/Characters/Player/Female/Medium/Bodies/F_MED_Pennant_Seas_One/Meshes/F_MED_Pennant_Seas_One_Physics
bodies {
 dyn_zipper [{
   shapeType=Capsule
   length=0.755
   radius=0.755
   shapeName=dyn_zipper_capsule
   center=<1.8221 0.0004 -0.5498>
   rotation=<rot 0 -90.0267 -87.9712>
  }]
 dyn_zipper_MSTR [{shapeType=Capsule, length=0.25, radius=0.505, shapeName=dyn_zipper_MSTR_capsule, rotation=<rot 0 -90 0>}]
 neck_01 [{
   shapeType=Capsule
   length=5.5365
   radius=4.4758
   shapeName=neck_01_capsule
   center=<0.2121 0.9755 -0.3821>
   rotation=<rot 28.5702 -101.4009 -86.9789>
  }]
 spine_05 [{
   shapeType=Capsule
   length=12.7879
   radius=6.3094
   shapeName=spine_05_capsule
   center=<3.16 -0.5801 0.0002>
   rotation=<rot 0 32.7202 -179.9967>
  }]
}
constraints [
 {bone1Name=neck_01, bone2Name=spine_05, swing1LimitDegrees=45, swing2LimitDegrees=45, twistLimitDegrees=45}
 {bone1Name=dyn_zipper_MSTR, bone2Name=neck_01, swing1LimitDegrees=45, swing2LimitDegrees=45, twistLimitDegrees=45}
 {bone1Name=dyn_zipper, bone2Name=dyn_zipper_MSTR, swing1LimitDegrees=30, swing2LimitDegrees=20, twistLimitDegrees=5}
]
