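# Post-copy animation Blueprint for the F_MED_Pennant_Seas_One female player body, adding deformation and secondary motion to the pose copied from its attached parent mesh.
# A deform Control Rig feeds constraints chaining the zipper bones to neck_01.
# A RigidBody node then simulates the override physics asset relative to spine_05, before a second Control Rig outputs the pose.
# An AnimDynamics zipper node sits unconnected, and the event graph has no logic.
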AnimBlueprint /Game/Characters/Player/Female/Medium/Bodies/F_MED_Pennant_Seas_One/Meshes/F_MED_Pennant_Seas_One_AnimBP
target_skeleton=&/F_MED_Pennant_Seas_One_Skeleton
graphs [
 AnimGraph {nodes [
   Anim Dynamics - Bone: dyn_zipper(Alpha=1) {
    bGravityOverrideInSimSpace=true
    bUseGravityOverride=true
    simulationSpace=World
    alphaBoolBlend {blendOption=Linear}
    boundBone {boneName=dyn_zipper}
    componentAppliedLinearAccClamp=<3000 3000 3000>
    componentLinearAccScale=<0.9 0.9 0.9>
    gravityOverride {z=-980}
    physicsBodyDefinitions [{
      boundBone {boneName=dyn_zipper}
      boxExtents=<10 10 10>
      constraintSetup {angularLimitsMax=<5 5 5>, angularLimitsMin=<-5 -5 -5>, angularTarget=<0 0 0>}
      localJointOffset=<0 0 0>
     }]
    relativeSpaceBone {boneName=neck_01}
   }
   n1 = Control Rig(Source=Copy Pose From Mesh() {bUseAttachedParent=true}) {
    alphaBoolBlend {blendOption=Linear}
    controlRigAssetReference {blueprintRigClass=&/F_MED_Pennant_Seas_One_DeformRig_ControlRig.F_MED_Pennant_Seas_One_DeformRig_ControlRig_C}
   }
   n2 = "Constraint - dyn_zipper_MSTR (P)"!(ComponentPose=Local To Component(LocalPose=@n1), Alpha=1, ConstraintWeights_0=1, ConstraintWeights_1:float=1) -> (Pose) {
    alphaBoolBlend {blendOption=Linear}
    boneToModify {boneName=dyn_zipper_MSTR}
    constraintSetup [{targetBone {boneName=neck_01}}]
    constraintWeights [1,0]
   }
   n3 = "Constraint - dyn_zipper (P)"!(ComponentPose=@n2, Alpha=1, ConstraintWeights_0=1, ConstraintWeights_1:float=1) -> (Pose) {
    alphaBoolBlend {blendOption=Linear}
    boneToModify {boneName=dyn_zipper}
    constraintSetup [{targetBone {boneName=dyn_zipper_MSTR}}]
    constraintWeights [1,0]
   }
   n4 = RigidBody(ComponentPose=@n3) {
    bClampLinearTranslationLimitToRefPose=true
    bOverrideWorldGravity=true
    overridePhysicsAsset=&/F_MED_Pennant_Seas_One_Physics
    simulationSpace=BaseBoneSpace
    alphaBoolBlend {blendOption=Linear}
    baseBoneRef {boneName=spine_05}
    componentAppliedLinearAccClamp=<3000 3000 3000>
    componentLinearAccScale=<0.7 0.7 0.7>
    componentLinearVelScale=<0.2 0.2 0.2>
    overrideWorldGravity {z=-980}
    simSpaceSettings {maxAngularAcceleration=1000, maxAngularVelocity=100, maxLinearAcceleration=250, maxLinearVelocity=5, externalLinearDragV=<20 20 20>}
   }
   n5 = Control Rig(Source=Component To Local(ComponentPose=@n4)) {
    alphaBoolBlend {blendOption=Linear}
    controlRigAssetReference {blueprintRigClass=&/F_MED_Pennant_Seas_One_CtrlRig.F_MED_Pennant_Seas_One_CtrlRig_C}
   }
   Output Pose(Result=@n5) {layerGroup=None}
  ]}
 EventGraph {nodes [Event BlueprintUpdateAnimation()]}
]
variables []
